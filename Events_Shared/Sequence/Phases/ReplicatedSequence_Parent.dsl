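; ReplicatedSequence_Parent is a parent Blueprint for event sequence phases that controls burn-in display on its sequence player.
; On BeginPlay it evaluates the ShowBurnInCurve curve table row and shows the burn-in at 1.0 or above, otherwise hides it.
; The SmartTunaToggleBurnIn custom event shows or hides the burn-in at runtime from a boolean.
; Tick and the construction script are empty.
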(event Custom|SmartTunaToggleBurnIn
  (params (BurnInEnabled :type Boolean))
  (if BurnInEnabled
    (Sequencer|Player|ShowBurnin)
    (else
      (Sequencer|Player|HideBurnin))))

(event EventBeginPlay
  (bind _outxy (Fortnite|EvaluateCurveTableRow (Variables|Default|GetShowBurnInCurve)))
  (if (>= _outxy 1.0)
    (Sequencer|Player|ShowBurnin)
    (else
      (Sequencer|Player|HideBurnin))))

(event EventTick
  (params (DeltaSeconds :type Float%20%28single-precision%29)))

(fn ConstructionScript)
